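# Sound cue for a fountain water spray at an Asteria POI, using desert city fountain attenuation and the ambient water sound class.
# The first two quality levels mix a random burst with two briefly delayed random start layers, each modulated and drawn from separate sample sets.
# The other two quality levels play just one modulated start sample.
# A concurrency override allows one instance, stopping the quietest, and virtualization is disabled.
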SoundCue /Game/Sounds/Asteria/POI/Fountain/SC_Fountain_Water_Spray
root=QualityLevel_0
nodes [
 {id=Delay_1, children [Modulator_0], properties {delayMax=0.25, delayMin=0.25}}
 {id=Delay_2, children [Modulator_2], properties {delayMax=0.25, delayMin=0.25}}
 {id=Mixer_0, children [Modulator_3,Delay_1,Delay_2], properties {inputVolume [2,2,1]}}
 {id=Modulator_0, children [Random_0]}
 {id=Modulator_1, children [WavePlayer_10]}
 {id=Modulator_2, children [Random_1]}
 {id=Modulator_3, children [Random_2]}
 {id=QualityLevel_0, children [Mixer_0,Mixer_0,Modulator_1,Modulator_1]}
 {id=Random_0, children [WavePlayer_7,WavePlayer_8,WavePlayer_9], properties {weights [1,1,1]}}
 {id=Random_1, children [WavePlayer_11, WavePlayer_12, WavePlayer_13], properties {weights [1,1,1]}}
 {id=Random_2, children [WavePlayer_0,WavePlayer_1,WavePlayer_2,WavePlayer_3], properties {weights [1,1,1,1]}}
 {id=WavePlayer_0, properties {soundWaveAssetPtr=&/Fountain_Water_Spray_Burst_01}}
 {id=WavePlayer_1, properties {soundWaveAssetPtr=&/Fountain_Water_Spray_Burst_02}}
 {id=WavePlayer_10, properties {soundWaveAssetPtr=&/Fountain_Water_Spray_Start_01}}
 {id=WavePlayer_11, properties {soundWaveAssetPtr=&/Fountain_Water_Spray_Start_04}}
 {id=WavePlayer_12, properties {soundWaveAssetPtr=&/Fountain_Water_Spray_Start_05}}
 {id=WavePlayer_13, properties {soundWaveAssetPtr=&/Fountain_Water_Spray_Start_06}}
 {id=WavePlayer_2, properties {soundWaveAssetPtr=&/Fountain_Water_Spray_Burst_03}}
 {id=WavePlayer_3, properties {soundWaveAssetPtr=&/Fountain_Water_Spray_Burst_04}}
 {id=WavePlayer_7, properties {soundWaveAssetPtr=&Fountain_Water_Spray_Start_01}}
 {id=WavePlayer_8, properties {soundWaveAssetPtr=&/Fountain_Water_Spray_Start_02}}
 {id=WavePlayer_9, properties {soundWaveAssetPtr=&/Fountain_Water_Spray_Start_03}}
]
properties {
 attenuationSettings=&/ATT_DesertCity_Fountain_Water
 bBypassVolumeScaleForPriority=true
 bOverrideConcurrency=true
 maxDistance=4300
 soundClassObject=/Game/Sounds/FortSoundClassesAndModes/Fort_Ambient_Water
 virtualizationMode=Disabled
 volumeMultiplier=0.7
 concurrencyOverrides {maxCount=1, resolutionRule=StopQuietest}
}
