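# Niagara system for the DarkVoyagerStrip environment fountain, firing choreographed water jets with spray audio and splashes.
# User arrays set each jet's position index, delay and height, with particle IDs wrapped by burst count.
# Fountain mesh particles scale by height and animate material parameters, partly distance-driven, while GPU Splash adds sprites and meshes.
# Delays extend lifetime and hide particles through alpha rather than postponing spawn; a DEBUG emitter is disabled.
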NiagaraSystem /Game/Effects/Fort_Effects/Effects/Environment/Hera/DarkVoyagerStrip/NS_DarkVoyagerStrip_Fountain
dependencies {
 usedDataInterfaces [
  &NiagaraDataInterfaceCurve
  &NiagaraDataInterfaceSpriteRendererInfo
  &NiagaraDataInterfaceCamera
  &NiagaraDataInterfaceAudioPlayer
 ]
 usedDynamicInputs [
  /Niagara/DynamicInputs/Math/GetInt32ArrayCount
  &DistanceBetweenPositions
  &MakeVectorFromLinearColorRGB
 ]
 usedModules [
  &self:SystemSpawnScript.SystemScriptSource.SystemScriptGraph.NiagaraNodeAssignment_0.SetVariables_#1
  &self:SystemSpawnScript.SystemScriptSource.SystemScriptGraph.NiagaraNodeAssignment_1.SetVariables_#2
  &SystemState
  &self:SystemSpawnScript.SystemScriptSource.SystemScriptGraph.NiagaraNodeAssignment_2.SetVariables_#3
  &self:SystemSpawnScript.SystemScriptSource.SystemScriptGraph.NiagaraNodeAssignment_3.SetVariables_#4
 ]
}
emitters {
 DEBUG {
  bEnabled=false
  emitterUpdateScript [
   &EmitterState()
   &SpawnBurst_Instantaneous(Spawn Count=$System.BurstCount, Use Spawn Probability=false, UseLoopCountLimit=false)
  ]
  particleSpawnScript [
   /SetVariables(Particles.DEBUGIndex:NiagaraInt32=&ModuloInt(A=&ReturnParticleIDAttributes(), B=$System.BurstCount)) {isSetParameters=true, module=SetVariables_#5}
   /SetVariables(Particles.PatternIndex:NiagaraInt32=&SelectIntFromArray(Direct Array Index=$Particles.DEBUGIndex, Int Selection Array=$User.PatternArray)) {isSetParameters=true, module=SetVariables_#6}
   /SetVariables(Particles.Delay:NiagaraFloat=&SelectFloatFromArray(Direct Array Index=$Particles.DEBUGIndex, Float Selection Array=$User.DelayArray)) {isSetParameters=true, module=SetVariables_#7}
   &InitializeParticle(
    Color=<rgba 0.263 1 0 1>
    Lifetime=&Add_Float(A=8, B=$Particles.Delay)
    Position=/Niagara/DynamicInputs/Arrays/SelectPositionFromArray(
     Array Sampling Mode:ENiagaraArraySamplingMode=::NewEnumerator1
     Direct Array Index:NiagaraInt32=$Particles.PatternIndex
     Interpolation:NiagaraFloat=?
     Direct Array Mode:ENiagara_WrapClamp=::NewEnumerator0
     Evaluation Type:ENiagaraRandomnessEvaluation=::NewEnumerator0
     Normalized Interpolation:NiagaraBool=true
     Recalculate Random Each Loop:NiagaraBool=false
     Position Array:NiagaraDataInterfaceArrayPosition=$User.PosArray
     Randomness Mode:ENiagaraRandomnessMode=::NewEnumerator0
     Fixed Random Seed:NiagaraBool=false
     Random Seed:NiagaraInt32=?
    )
    Position Mode=::NewEnumerator1
    Sprite Size Mode=::NewEnumerator3
    Uniform Sprite Size=50
    UsePositionOffset=false
    Write Material Random=true
   )
   &CameraOffset(Camera Offset Amount=25)
  ]
  particleUpdateScript [
   &ParticleState()
   /SetVariables(Particles.FakeNormalizedAge:NiagaraFloat=&RemapRange(Input Value To Remap=$Particles.Age, Input Min=$Particles.Delay, Input Max=$Particles.Lifetime, Output Min=0, Output Max=1)) {isSetParameters=true, module=SetVariables_#8}
   &ScaleColor(
    Scale Alpha=&MakeCustomFloatFromBool(
     Bool=&SetBoolByFloatComparison(A=$Particles.Age, B=$Particles.Delay, Comparison Type=::NewEnumerator7)
     True Float=0
     False Float=1
    )
    ScaleA=true
    ScaleRGB=false
   ) {module=ScaleColor001}
   &ScaleColor(
    Scale Alpha=&FloatFromCurve(FloatCurve=DataInterface() {propertyValues {Curve {keys [<at 0 1>,<at 1 0>]}}}, CurveIndex=$Particles.FakeNormalizedAge)
    ScaleA=true
    ScaleRGB=false
   )
  ]
  rendererClasses [&NiagaraSpriteRendererProperties]
  renderers [{rendererClass=&NiagaraSpriteRendererProperties}]
 }
 Fountain {
  emitterSpawnScript [&InitializeAudioPlayer(Audio Player=DataInterface() {propertyValues {
      Attenuation=null
      Concurrency=null
      MaxPlaysPerTick=10
      SoundToPlay=/Game/Sounds/Asteria/POI/Fountain/SC_Fountain_Water_Spray
      bAllowLoopingOneShotSounds=false
      bLimitPlaysPerTick=true
      bOnlyActiveDuringGameplay=false
      bStopWhenComponentIsDestroyed=true
      ConfigurationUserParameter {parameter {name=None, typeDefHandle {registeredTypeIndex=1}, varData []}}
      ParameterNames []
     }})]
  emitterUpdateScript [
   &EmitterState()
   &SpawnBurst_Instantaneous(Spawn Count=$System.BurstCount, Use Spawn Probability=false, UseLoopCountLimit=false)
  ]
  particleSpawnScript [
   /SetVariables(Particles.Index:NiagaraInt32=&ModuloInt(A=&ReturnParticleIDAttributes(), B=$System.BurstCount)) {isSetParameters=true, module=SetVariables_#5}
   /SetVariables(Particles.PatternIndex:NiagaraInt32=&SelectIntFromArray(Direct Array Index=$Particles.Index, Int Selection Array=$User.PatternArray)) {isSetParameters=true, module=SetVariables_#6}
   /SetVariables(Particles.Delay:NiagaraFloat=&SelectFloatFromArray(Direct Array Index=$Particles.Index, Float Selection Array=$User.DelayArray)) {isSetParameters=true, module=SetVariables_#7}
   /SetVariables(Particles.Height:NiagaraFloat=&SelectFloatFromArray(Direct Array Index=$Particles.Index, Float Selection Array=$User.HeightArray)) {isSetParameters=true}
   &InitializeParticle(
    Color Mode=::NewEnumerator0
    Lifetime=&Add_Float(A=&Lerp_Float(A=0.25, B=6, Alpha=$Particles.Height), B=$Particles.Delay)
    Mesh Scale=&Multiply_VectorByFloat(Vector=<1 1 15>, Float=$Particles.Height)
    Mesh Scale Mode=::NewEnumerator4
    Position=&SelectPositionFromArray(
     Array Sampling Mode:ENiagaraArraySamplingMode=::NewEnumerator1
     Direct Array Index:NiagaraInt32=$Particles.PatternIndex
     Interpolation:NiagaraFloat=?
     Direct Array Mode:ENiagara_WrapClamp=::NewEnumerator0
     Evaluation Type:ENiagaraRandomnessEvaluation=::NewEnumerator0
     Normalized Interpolation:NiagaraBool=true
     Recalculate Random Each Loop:NiagaraBool=false
     Position Array:NiagaraDataInterfaceArrayPosition=$User.PosArray
     Randomness Mode:ENiagaraRandomnessMode=::NewEnumerator0
     Fixed Random Seed:NiagaraBool=false
     Random Seed:NiagaraInt32=?
    )
    Position Mode=::NewEnumerator1
    UsePositionOffset=false
    Write Material Random=true
   )
   /Niagara/Modules/Audio/PlayAudio(Audio Playback Position:NiagaraPosition=&AddVectorToPosition(Position=$Particles.Position, Vector=<0 0 200>))
   /SetVariables() {isSetParameters=true}
   /SetVariables() {isSetParameters=true}
  ]
  particleUpdateScript [
   &ParticleState()
   /SetVariables(Particles.NormalizedAge:NiagaraFloat=&RemapRange(Input Value To Remap=$Particles.Age, Input Min=$Particles.Delay, Input Max=$Particles.Lifetime, Output Min=0, Output Max=1)) {isSetParameters=true, module=SetVariables_#8}
   /SetVariables(Particles.NoisePanStream:NiagaraFloat=&Add_Float(
     A=$Particles.NoisePanStream
     B=&Multiply_Float(
      A=&FloatFromCurve(
       FloatCurve=DataInterface() {propertyValues {Curve {defaultValue=1, keys [{interpMode=RCIM_Cubic, tangentMode=RCTM_User, time=0, value=1}, {tangentMode=RCTM_User, time=0.5, value=-0.5}]}}}
       Scale Curve=1.5
      )
      B=$Engine.DeltaTime
     )
    )) {isSetParameters=true}
   /SetVariables(Particles.NoisePanMist:NiagaraFloat=&Add_Float(
     A=$Particles.NoisePanMist
     B=&Multiply_Float(
      A=&FloatFromCurve(
       FloatCurve=DataInterface() {propertyValues {Curve {
          defaultValue=1
          keys [
           {arriveTangent=-2.9107, interpMode=RCIM_Cubic, leaveTangent=-2.9107, tangentMode=RCTM_User, time=0, value=1}
           {interpMode=RCIM_Cubic, tangentMode=RCTM_User, time=1, value=0}
          ]
         }}}
       Scale Curve=0.5
      )
      B=$Engine.DeltaTime
     )
    )) {isSetParameters=true}
   &ScaleColor(
    Scale Alpha=&MakeCustomFloatFromBool(
     Bool=&SetBoolByFloatComparison(A=$Particles.Age, B=$Particles.Delay, Comparison Type=::NewEnumerator7)
     True Float=0
     False Float=1
    )
    ScaleA=true
    ScaleRGB=false
   ) {module=ScaleColor001}
   &ScaleColor(
    Scale Alpha=&FloatFromCurve(FloatCurve=DataInterface() {propertyValues {Curve {
        defaultValue=1
        keys [{interpMode=RCIM_Cubic, tangentMode=RCTM_User, time=0, value=1}, {interpMode=RCIM_Cubic, tangentMode=RCTM_User, time=0.5, value=0}]
       }}})
    ScaleA=true
    ScaleRGB=false
   )
   &DynamicMaterialParameters(
    Index 0 Param 1=1
    Index 0 Param 2=$Particles.Height
    Index 0 Param 3=$Particles.NoisePanMist
    Index 0 Param 4=&FloatFromCurve(FloatCurve=DataInterface() {propertyValues {Curve {
        defaultValue=0
        keys [{interpMode=RCIM_Cubic, tangentMode=RCTM_User, time=0, value=0}, {interpMode=RCIM_Cubic, tangentMode=RCTM_User, time=1, value=1}]
       }}})
    Index 1 Param 1=&Lerp_Float(
     A=2
     B=40
     Alpha=&FloatFromCurve(
      FloatCurve=DataInterface() {propertyValues {Curve {
         defaultValue=0
         keys [
          {interpMode=RCIM_Cubic, tangentMode=RCTM_User, time=0, value=0}
          {arriveTangent=2.4939, interpMode=RCIM_Cubic, leaveTangent=2.4939, tangentMode=RCTM_User, time=1, value=1}
         ]
        }}}
      Scale Curve=&Lerp_Float(A=1, B=3, Alpha=$System.NormalizedDistance)
     )
    )
    Index 1 Param 2=&FloatFromCurve(FloatCurve=DataInterface() {propertyValues {Curve {
        defaultValue=0
        keys [
         {arriveTangent=5.795, interpMode=RCIM_Cubic, leaveTangent=5.795, tangentMode=RCTM_User, time=0, value=0}
         {arriveTangent=1.8993, interpMode=RCIM_Cubic, leaveTangent=1.8993, tangentMode=RCTM_User, time=0.1525, value=0.8805}
         {arriveTangent=-2.9932, interpMode=RCIM_Cubic, leaveTangent=-2.9932, tangentMode=RCTM_User, time=0.4751, value=0.691}
         {arriveTangent=0.1838, interpMode=RCIM_Cubic, leaveTangent=0.1838, tangentMode=RCTM_User, time=1, value=0}
        ]
       }}})
    Index 1 Param 3=1
    Index 1 Param 4=&Lerp_Float(
     A=0
     B=3
     Alpha=&FloatFromCurve(FloatCurve=DataInterface() {propertyValues {Curve {
         defaultValue=0
         keys [
          {interpMode=RCIM_Cubic, tangentMode=RCTM_User, time=0, value=0}
          {arriveTangent=2.4939, interpMode=RCIM_Cubic, leaveTangent=2.4939, tangentMode=RCTM_User, time=1, value=1}
         ]
        }}})
    )
    Index 2 Linear Color:LinearColor=<rgba 5 2 0.5 1>
    Module.Index0Float0Write=true
    Module.Index0Float1Write=true
    Module.Index0Float2Write=true
    Module.Index0Float3Write=true
    Module.Index1Float0Write=true
    Module.Index1Float1Write=true
    Module.Index1Float2Write=true
    Module.Index1Float3Write=true
    Write Parameter Index 1=::NewEnumerator1
    Write Parameter Index 2=::NewEnumerator5
   )
   /SetVariables(Particles.DMPStream:Vector4f=&MakeVector4(
     X=1
     Y=$Particles.Height
     Z=$Particles.NoisePanStream
     W=&FloatFromCurve(FloatCurve=DataInterface() {propertyValues {Curve {
         defaultValue=0
         keys [{interpMode=RCIM_Cubic, tangentMode=RCTM_User, time=0, value=0}, {interpMode=RCIM_Cubic, tangentMode=RCTM_User, time=1, value=1}]
        }}})
    )) {isSetParameters=true}
   /SetVariables(Particles.DMPStream1:Vector4f=&MakeVector4(
     X=&Lerp_Float(
      A=2
      B=40
      Alpha=&FloatFromCurve(
       FloatCurve=DataInterface() {propertyValues {Curve {
          defaultValue=0
          keys [
           {interpMode=RCIM_Cubic, tangentMode=RCTM_User, time=0, value=0}
           {arriveTangent=2.4939, interpMode=RCIM_Cubic, leaveTangent=2.4939, tangentMode=RCTM_User, time=1, value=1}
          ]
         }}}
       Scale Curve=&Lerp_Float(A=1, B=3, Alpha=$System.NormalizedDistance)
      )
     )
     Y=&FloatFromCurve(FloatCurve=DataInterface() {propertyValues {Curve {
         defaultValue=0
         keys [
          {arriveTangent=5.795, interpMode=RCIM_Cubic, leaveTangent=5.795, tangentMode=RCTM_User, time=0, value=0}
          {arriveTangent=1.8993, interpMode=RCIM_Cubic, leaveTangent=1.8993, tangentMode=RCTM_User, time=0.1525, value=0.8805}
          {arriveTangent=-2.9932, interpMode=RCIM_Cubic, leaveTangent=-2.9932, tangentMode=RCTM_User, time=0.4751, value=0.691}
          {arriveTangent=0.1838, interpMode=RCIM_Cubic, leaveTangent=0.1838, tangentMode=RCTM_User, time=1, value=0}
         ]
        }}})
     Z=&Lerp_Float(
      A=1
      B=30
      Alpha=&FloatFromCurve(FloatCurve=DataInterface() {propertyValues {Curve {
          defaultValue=0
          keys [
           {interpMode=RCIM_Cubic, tangentMode=RCTM_User, time=0, value=0}
           {arriveTangent=2.4939, interpMode=RCIM_Cubic, leaveTangent=2.4939, tangentMode=RCTM_User, time=1, value=1}
          ]
         }}})
     )
     W=0.1
    )) {isSetParameters=true}
  ]
  rendererClasses [&NiagaraMeshRendererProperties]
  renderers [
   {rendererClass=&NiagaraMeshRendererProperties, rendererIndex=0}
   {rendererClass=&NiagaraMeshRendererProperties, rendererIndex=1}
   {rendererClass=&NiagaraMeshRendererProperties, rendererIndex=2}
   {rendererClass=&NiagaraMeshRendererProperties, rendererIndex=3}
  ]
 }
 Splash {
  simTarget=GPUComputeSim
  emitterUpdateScript [
   &EmitterState(
    Life Cycle Mode=::NewEnumerator1
    Loop Behavior=::NewEnumerator1
    Loop Duration=&Add_Float(A=-0.25, B=$User.LongestDelay)
    UseLoopDelay=false
   )
   &SpawnBurst_Instantaneous(Spawn Count=&Multiply_Int(A=3, B=$System.BurstCount), Use Spawn Probability=false, UseLoopCountLimit=false)
   &SpawnRate(SpawnRate=&Multiply_FloatByInt(Float=3, Integer=$System.BurstCount), Use Spawn Probability=false)
  ]
  particleSpawnScript [
   &CameraOffset(Camera Offset Amount=25)
   /SetVariables(Particles.Index:NiagaraInt32=&ModuloInt(A=&ReturnParticleIDAttributes(), B=$System.BurstCount)) {isSetParameters=true}
   /SetVariables(Particles.PatternIndex:NiagaraInt32=&SelectIntFromArray(Direct Array Index=$Particles.Index, Int Selection Array=$User.PatternArray)) {isSetParameters=true}
   /SetVariables(Particles.Delay:NiagaraFloat=&SelectFloatFromArray(Direct Array Index=$Particles.Index, Float Selection Array=$User.DelayArray)) {isSetParameters=true}
   &InitializeParticle(
    Color=<rgba 1 0.2348 0.8725 1>
    Lifetime=&Add_Float(A=&RandomRangeFloat(Minimum=0.6, Maximum=1), B=$Particles.Delay)
    Mesh Scale Mode=::NewEnumerator1
    Mesh Uniform Scale Max=3
    Mesh Uniform Scale Min=1.5
    Position=&SelectPositionFromArray(
     Array Sampling Mode:ENiagaraArraySamplingMode=::NewEnumerator1
     Direct Array Index:NiagaraInt32=$Particles.PatternIndex
     Interpolation:NiagaraFloat=?
     Direct Array Mode:ENiagara_WrapClamp=::NewEnumerator0
     Evaluation Type:ENiagaraRandomnessEvaluation=::NewEnumerator0
     Normalized Interpolation:NiagaraBool=true
     Recalculate Random Each Loop:NiagaraBool=false
     Position Array:NiagaraDataInterfaceArrayPosition=$User.PosArray
     Randomness Mode:ENiagaraRandomnessMode=::NewEnumerator0
     Fixed Random Seed:NiagaraBool=false
     Random Seed:NiagaraInt32=?
    )
    Position Mode=::NewEnumerator1
    Sprite Size Max=<50 50>
    Sprite Size Min=<25 25>
    Sprite Size Mode=::NewEnumerator2
    UsePositionOffset=false
    Write Material Random=true
   )
   &AddVelocityInCone(
    Cone Angle=5
    Cone Axis=<0 0 1>
    Module.Use Velocity Falloff On Cone Axis=false
    Velocity Distribution Along Cone Axis=0.25
    Velocity Strength=&UniformRangedFloat(Minimum=500, Maximum=50)
   )
   &SubUVAnimation(
    End Frame Range Override=3
    "Sprite Renderer "=DataInterface() {propertyValues {SpriteRenderer=&self:Splash_0.NiagaraSpriteRendererProperties_1}}
    SubUV Animation Mode:ENiagara_SubUVLookupModeV2=::NewEnumerator2
    UseEndFrame=true
    UseStartFrame=true
   )
   &InitialMeshOrientation(
    Mesh Orientation Mode=::NewEnumerator1
    Rotation=&RandomRangeVector(Minimum=<0 0 0>, Maximum=<0 0 1>)
    Use Rotation Vector=true
   )
   &ShapeLocation(MaskSpawnGroup=false, Offset=<0 0 10>, Offset Mode=::NewEnumerator0, Override Seed=false, Sphere Radius=6, UseOwnerScale=true)
  ]
  particleUpdateScript [
   &ParticleState()
   &ScaleMeshSize(Scale Factor=&VectorFromFloat(Value=2))
   /SetVariables(Particles.NormalizedAge:NiagaraFloat=&RemapRange(Input Value To Remap=$Particles.Age, Input Min=$Particles.Delay, Input Max=$Particles.Lifetime, Output Min=0, Output Max=1)) {isSetParameters=true}
   &ScaleSpriteSize(Scale Factor=&Vector2DFromFloat(Value=&FloatFromCurve(
      FloatCurve=DataInterface() {propertyValues {Curve {keys [
          {interpMode=RCIM_Cubic, tangentMode=RCTM_User, time=0, value=0}
          {arriveTangent=1.4313, interpMode=RCIM_Cubic, leaveTangent=1.4313, tangentMode=RCTM_User, time=0.6, value=2}
          {tangentMode=RCTM_User, time=1, value=1.5}
         ]}}}
      Scale Curve=2
     )))
   &DynamicMaterialParameters(
    Index 0 Param 1=&FloatFromCurve(
     FloatCurve=DataInterface() {propertyValues {Curve {keys [
         {arriveTangent=-0.5095, interpMode=RCIM_Cubic, leaveTangent=-0.5095, tangentMode=RCTM_User, time=0, value=1}
         {arriveTangent=-1.3585, interpMode=RCIM_Cubic, leaveTangent=-1.3585, tangentMode=RCTM_User, time=1, value=0}
        ]}}}
     Scale Curve=1.5
    )
    Module.Index0Float0Write=true
    Module.Index0Float1Write=true
    Module.Index0Float2Write=true
    Module.Index0Float3Write=true
   )
   &ScaleColor(
    Scale Alpha=&MakeCustomFloatFromBool(
     Bool=&SetBoolByFloatComparison(A=$Particles.Age, B=$Particles.Delay, Comparison Type=::NewEnumerator7)
     True Float=0
     False Float=1
    )
    ScaleA=true
    ScaleRGB=false
   ) {module=ScaleColor001}
   &SpriteRotationRate(Rotation Rate=3)
   &GravityForce(Gravity=<0 0 -100>)
   &SolveForcesAndVelocity(Clamp Velocity=true, Limit Acceleration=false, Speed Limit=5000)
   /SetVariables(
    Particles.Position2:NiagaraPosition=&AddVectorToPosition(Position=$Particles.Position, Vector=<0 0 0>)
    Particles.Velocity2:Vector3f=<0 0 1>
   ) {isSetParameters=true}
   &SolveForcesAndVelocity(
    Clamp Velocity=true
    Limit Acceleration=false
    Position:NiagaraPosition=$Particles.Position2
    Speed Limit=5000
    Velocity:Vector3f=$Particles.Velocity2
   ) {module=SolveForcesAndVelocity001}
   &ScaleColor(
    Scale Alpha=&FloatFromCurve(
     FloatCurve=DataInterface() {propertyValues {Curve {
        defaultValue=0
        keys [
         {interpMode=RCIM_Cubic, tangentMode=RCTM_User, time=0, value=0}
         {interpMode=RCIM_Cubic, tangentMode=RCTM_User, time=0.5, value=1}
         {tangentMode=RCTM_User, time=1, value=0}
        ]
       }}}
     Scale Curve=3
    )
    ScaleA=true
    ScaleRGB=false
   )
   &ScaleColor(
    Scale Alpha=/Niagara/DynamicInputs/Camera/CameraFade(
     Camera Query:NiagaraDataInterfaceCamera=DataInterface()
     Fade Offset:NiagaraFloat=500
     Fade Distance:NiagaraFloat=500
     Use Owner Position:NiagaraBool=true
    )
    ScaleA=true
    ScaleRGB=true
   ) {module=ScaleColor002}
  ]
  rendererClasses [&NiagaraSpriteRendererProperties, &NiagaraMeshRendererProperties]
  renderers [
   {rendererClass=&NiagaraSpriteRendererProperties}
   {rendererClass=&NiagaraSpriteRendererProperties, rendererIndex=1}
   {rendererClass=&NiagaraMeshRendererProperties, rendererIndex=2}
  ]
 }
}
summary {
 systemName=NS_DarkVoyagerStrip_Fountain
 emitters [
  {bEnabled=false, emitterName=DEBUG, rendererClasses [&NiagaraSpriteRendererProperties]}
  {
   emitterName=Splash
   simTarget=GPUComputeSim
   rendererClasses [&NiagaraSpriteRendererProperties, &NiagaraMeshRendererProperties]
  }
  {emitterName=Fountain, rendererClasses [&NiagaraMeshRendererProperties]}
 ]
 userVariables [
  User.DelayArray {
   type=NiagaraDataInterfaceArrayFloat
   defaultValue {
    struct=NiagaraExt_VariableValue_DataInterface
    value {
     dataInterface=&self:NiagaraDataInterfaceArrayFloat_0
     dataInterfaceClass=&NiagaraDataInterfaceArrayFloat
    }
   }
  }
  User.HeightArray {
   type=NiagaraDataInterfaceArrayFloat
   defaultValue {
    struct=NiagaraExt_VariableValue_DataInterface
    value {
     dataInterface=&self:NiagaraDataInterfaceArrayFloat_1
     dataInterfaceClass=&NiagaraDataInterfaceArrayFloat
    }
   }
  }
  User.LongestDelay {type=NiagaraFloat, defaultValue=0.5, description=""}
  User.PatternArray {
   type=NiagaraDataInterfaceArrayInt32
   description=""
   defaultValue {
    struct=NiagaraExt_VariableValue_DataInterface
    value {
     dataInterface=&self:NiagaraDataInterfaceArrayInt32_0
     dataInterfaceClass=&NiagaraDataInterfaceArrayInt32
    }
   }
  }
  User.PosArray {
   type=NiagaraDataInterfaceArrayPosition
   description=""
   defaultValue {
    struct=NiagaraExt_VariableValue_DataInterface
    value {
     dataInterface=&self:NiagaraDataInterfaceArrayPosition_0
     dataInterfaceClass=&NiagaraDataInterfaceArrayPosition
    }
   }
  }
 ]
}
